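# Physics asset for the Ocean Rider backpack cosmetic, defining bodies and joints for its dangling accessories: zippers, carabiners, cork bottle, rope, shovel and compass needle.
# Accessories are box bodies, except a capsule shovel, joined by angle-limited constraints to small base bodies, which have no constraint to the pack body.
# Cork string and rope form three-link chains off the bottle and rope carabiner.
# Pocket zippers chain from the cork bottle base, and the needle hangs off the shovel with Swing2 locked.
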PhysicsAsset /BRCosmetics/Accessories/FORT_Backpacks/Backpack_F_MED_Ocean_Rider/Meshes/F_MED_Ocean_Rider_Pack_Physics
bodies {
 dyn_bag_zipper_base [{shapeType=Box, extentX=0.5, extentY=0.5, extentZ=0.5}]
 dyn_bag_zipper_parent [{shapeType=Box, extentX=3.76, extentY=2.76, extentZ=1.01, center=<2.0292 0 0>}]
 dyn_bottle_carabiner_base [{shapeType=Box, extentX=0.5, extentY=0.5, extentZ=0.5}]
 dyn_bottle_carabiner_parent [{shapeType=Box, extentX=5.9846, extentY=3.3606, extentZ=1.3921, center=<3.2541 0 0>}]
 dyn_compass_needle [{shapeType=Box, extentX=1.76, extentY=0.75, extentZ=0.76, center=<0.7146 0 0>}]
 dyn_cork_bottle_base [{shapeType=Box, extentX=0.5, extentY=0.5, extentZ=0.5}]
 dyn_cork_bottle_parent [{shapeType=Box, extentX=6, extentY=3.75, extentZ=3.9302, center=<2.8288 0 0>}]
 dyn_cork_string_01 [{shapeType=Box, extentX=2.26, extentY=1.76, extentZ=1.76, center=<1.1707 0 0>}]
 dyn_cork_string_02 [{shapeType=Box, extentX=3.25, extentY=1.75, extentZ=1.9829, center=<1.8146 0 0>}]
 dyn_cork_string_03 [{shapeType=Box, extentX=2.8416, extentY=2.8538, extentZ=2.6587, center=<1.2582 0 0>}]
 dyn_pocket_zipper [{shapeType=Box, extentX=3.9964, extentY=2.9551, extentZ=1.07, center=<2.0144 0 0.0001>}]
 dyn_pocket_zipper_01 [{shapeType=Box, extentX=4.9255, extentY=3.164, extentZ=1.1286, center=<2.1472 0 0>}]
 dyn_rope_01 [{shapeType=Box, extentX=3.01, extentY=4.76, extentZ=2.76, center=<1.3538 0 0>}]
 dyn_rope_02 [{shapeType=Box, extentX=4.8948, extentY=4.888, extentZ=2.9175, center=<2.5256 0 0>}]
 dyn_rope_03 [{shapeType=Box, extentX=4.0605, extentY=4.8978, extentZ=4.8801, center=<1.9867 0 0>}]
 dyn_rope_carabiner_base [{shapeType=Box, extentX=0.5, extentY=0.5, extentZ=0.5}]
 dyn_rope_carabiner_parent [{shapeType=Box, extentX=7.606, extentY=3.3915, extentZ=3.817, center=<4.1981 0 0>, rotation=<rot 0 -7.3761 0>}]
 dyn_shovel_base [{shapeType=Box, extentX=0.5, extentY=0.5, extentZ=0.5}]
 dyn_shovel_parent [{shapeType=Capsule, length=8.25, radius=2.5, center=<6.5054 0 0>, rotation=<rot 90 0 0>}]
 pack [{shapeType=Box, extentX=1.75, extentY=1.75, extentZ=1.75}]
}
constraints [
 {
  bone1Name=dyn_bottle_carabiner_parent
  bone2Name=dyn_bottle_carabiner_base
  swing1LimitDegrees=30
  swing2LimitDegrees=40
  twistLimitDegrees=3
 }
 {bone1Name=dyn_cork_bottle_parent, bone2Name=dyn_cork_bottle_base, swing1LimitDegrees=10, swing2LimitDegrees=10, twistLimitDegrees=3}
 {bone1Name=dyn_cork_string_01, bone2Name=dyn_cork_bottle_parent, swing1LimitDegrees=20, swing2LimitDegrees=20, twistLimitDegrees=5}
 {bone1Name=dyn_cork_string_02, bone2Name=dyn_cork_string_01, swing1LimitDegrees=20, swing2LimitDegrees=20, twistLimitDegrees=5}
 {bone1Name=dyn_cork_string_03, bone2Name=dyn_cork_string_02, swing1LimitDegrees=20, swing2LimitDegrees=20, twistLimitDegrees=5}
 {bone1Name=dyn_bag_zipper_parent, bone2Name=dyn_bag_zipper_base, swing1LimitDegrees=20, swing2LimitDegrees=40, twistLimitDegrees=3}
 {bone1Name=dyn_rope_carabiner_parent, bone2Name=dyn_rope_carabiner_base, swing1LimitDegrees=6, swing2LimitDegrees=6, twistLimitDegrees=3}
 {bone1Name=dyn_rope_01, bone2Name=dyn_rope_carabiner_parent, swing1LimitDegrees=15, swing2LimitDegrees=15, twistLimitDegrees=5}
 {bone1Name=dyn_rope_02, bone2Name=dyn_rope_01, swing1LimitDegrees=30, swing2LimitDegrees=30, twistLimitDegrees=5}
 {bone1Name=dyn_rope_03, bone2Name=dyn_rope_02, swing1LimitDegrees=40, swing2LimitDegrees=40, twistLimitDegrees=5}
 {bone1Name=dyn_shovel_parent, bone2Name=dyn_shovel_base, swing1LimitDegrees=10, swing2LimitDegrees=5, twistLimitDegrees=2}
 {bone1Name=dyn_pocket_zipper, bone2Name=dyn_cork_bottle_base, swing1LimitDegrees=20, swing2LimitDegrees=20, twistLimitDegrees=3}
 {bone1Name=dyn_pocket_zipper_01, bone2Name=dyn_pocket_zipper, swing1LimitDegrees=20, swing2LimitDegrees=20, twistLimitDegrees=3}
 {
  bone1Name=dyn_compass_needle
  bone2Name=dyn_shovel_parent
  swing1LimitDegrees=5
  swing2LimitDegrees=10
  swing2Motion=Locked
  twistLimitDegrees=1
 }
]
